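# Animation Blueprint for the female medium soccer body cosmetic, following the owner's skeletal mesh with procedural shoulder correction.
# The copied pose passes through a Control Rig, then four additive bone-space translations nudge front and upper shoulder helper bones.
# Each update derives their alphas from arm-to-clavicle reference-pose deltas on the owner mesh.
# The right-side "up" alpha samples upperarm_twist_01_r, whereas the left samples upperarm_l.
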AnimBlueprint /BRCosmetics/Characters/Player/Female/Medium/Bodies/F_MED_Soccer/Meshes/F_MED_Soccer_01_AnimBP
target_skeleton=&/F_MED_Soccer_01_Skeleton
graphs [
 AnimGraph {nodes [
   n1 = Control Rig(Source=Copy Pose From Mesh(SourceMeshComponent=Get MeshToCopy() -> (MeshToCopy:Skeletal Mesh Component Object Reference))) {controlRigAssetReference {controlRigAsset=&/F_MED_Soccer_01_CtrlRig}}
   n2 = "Transform (Modify) Bone - Bone: flont_uppelalm_l"(
    ComponentPose=Local To Component(LocalPose=@n1)
    Alpha=Get Shoulder_rot_f_l() -> (Shoulder_rot_f_l:double)
    Translation="5,5,-5"
    Rotation="0,0,0"
    Scale="1,1,1"
   ) {
    rotationMode=BMM_Additive
    rotationSpace=BCS_BoneSpace
    scaleMode=BMM_Additive
    scaleSpace=BCS_BoneSpace
    translationMode=BMM_Additive
    translationSpace=BCS_BoneSpace
    alphaBoolBlend {blendOption=Linear}
    boneToModify {boneName=flont_uppelalm_l}
    translation=<5 5 -5>
   }
   n3 = "Transform (Modify) Bone - Bone: front_upperarm_r"(ComponentPose=@n2, Alpha=Get Shoulder_rot_f_r() -> (Shoulder_rot_f_r:double), Translation="-5,-5,5", Rotation="0,0,0", Scale="1,1,1") {
    rotationMode=BMM_Additive
    rotationSpace=BCS_BoneSpace
    scaleMode=BMM_Additive
    scaleSpace=BCS_BoneSpace
    translationMode=BMM_Additive
    translationSpace=BCS_BoneSpace
    alphaBoolBlend {blendOption=Linear}
    boneToModify {boneName=front_upperarm_r}
    translation=<-5 -5 5>
   }
   n4 = "Transform (Modify) Bone - Bone: up_uppelalm_l"(ComponentPose=@n3, Alpha=Get Shoulder_rot_up_l() -> (Shoulder_rot_up_l:double), Translation="0,2,2", Rotation="0,0,0", Scale="1,1,1") {
    rotationMode=BMM_Additive
    rotationSpace=BCS_BoneSpace
    scaleMode=BMM_Additive
    scaleSpace=BCS_BoneSpace
    translationMode=BMM_Additive
    translationSpace=BCS_BoneSpace
    alphaBoolBlend {blendOption=Linear}
    boneToModify {boneName=up_uppelalm_l}
    translation {y=2, z=2}
   }
   n5 = "Transform (Modify) Bone - Bone: up_upperarm_r"(ComponentPose=@n4, Alpha=Get Shoulder_rot_up_r() -> (Shoulder_rot_up_r:double), Translation="0,-2,-2", Rotation="0,0,0", Scale="1,1,1") {
    rotationMode=BMM_Additive
    rotationSpace=BCS_BoneSpace
    scaleMode=BMM_Additive
    scaleSpace=BCS_BoneSpace
    translationMode=BMM_Additive
    translationSpace=BCS_BoneSpace
    alphaBoolBlend {blendOption=Linear}
    boneToModify {boneName=up_upperarm_r}
    translation {y=-2, z=-2}
   }
   Output Pose(Result=Component To Local(ComponentPose=@n5)) {layerGroup=None}
  ]}
 EventGraph {nodes [
   Event BlueprintInitializeAnimation()
   > Set MeshToCopy(MeshToCopy:Skeletal Mesh Component Object Reference=GetOwnerSkeletalMeshComponent()) -> (Output_Get:Skeletal Mesh Component Object Reference)
   Event BlueprintUpdateAnimation()
   n1 = Get MeshToCopy()
   > Set Shoulder_rot_f_r(Shoulder_rot_f_r:double=GetAlphaFromReferenceTransformDelta(SkeletalMeshComponent=@n1, BoneName=upperarm_r, BaseName=clavicle_r, InputRange=(X=0,Y=90), TwistAxis="0,0,1")) -> (Output_Get:double)
   > Set Shoulder_rot_f_l(Shoulder_rot_f_l:double=GetAlphaFromReferenceTransformDelta(SkeletalMeshComponent=@n1, BoneName=upperarm_l, BaseName=clavicle_l, InputRange=(X=0,Y=90), TwistAxis="0,0,1")) -> (Output_Get:double)
   > Set Shoulder_rot_up_r(Shoulder_rot_up_r:double=GetAlphaFromReferenceTransformDelta(
     SkeletalMeshComponent=@n1
     BoneName=upperarm_twist_01_r
     BaseName=clavicle_r
     DeltaComponentType=EulerY
     InputRange=(X=0,Y=110)
     TwistAxis="0,0,1"
    )) -> (Output_Get:double)
   > Set Shoulder_rot_up_l(Shoulder_rot_up_l:double=GetAlphaFromReferenceTransformDelta(SkeletalMeshComponent=@n1, BoneName=upperarm_l, BaseName=clavicle_l, DeltaComponentType=EulerY, InputRange=(X=0,Y=110), TwistAxis="0,0,1")) -> (Output_Get:double)
   TryGetPawnOwner()
  ]}
]
variables [
 MeshToCopy {type=SkeletalMeshComponent}
 Shoulder_rot_f_l {type=number}
 Shoulder_rot_f_r {type=number}
 Shoulder_rot_up_l {type=number}
 Shoulder_rot_up_r {type=number}
 Skel_Mesh {type=SkeletalMeshComponent}
]
